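# Physics asset for the OceanPearlFNCS backpack cosmetic mesh, providing simplified collision bodies for its skeleton.
# Two capsule bodies approximate the mesh: a larger one on the pack bone and a smaller one on the laures bone.
# A single constraint links laures to pack, with matching angle values for both swing axes and twist.
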PhysicsAsset /BRCosmetics/Accessories/FORT_Backpacks/Backpack_M_MED_OceanPearlFNCS/Meshes/M_MED_OceanPearlFNCS_Pack_PhysicsAsset
bodies {
 laures [{shapeType=Capsule, length=16.6945, radius=8.298, center=<0 -0.2262 5.026>, rotation=<rot 0 -90 -90>}]
 pack [{shapeType=Capsule, length=19.4414, radius=14.9664, center=<-0.1607 -2.3325 -5.5601>, rotation=<rot 0 -89.9606 -89.0647>}]
}
constraints [{bone1Name=laures, bone2Name=pack, swing1LimitDegrees=45, swing2LimitDegrees=45, twistLimitDegrees=45}]
